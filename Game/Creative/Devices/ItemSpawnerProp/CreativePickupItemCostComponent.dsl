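; Cost component for Creative pickups, used with the Item Spawner device, that reports the pickup's resource cost.
; GetCost casts the owner to FortPickupCreative and, if its cost component is valid, reads that component's item-cost array by resource-type enum index.
; It returns a map of Wood, Stone, Metal and GoldCurrency costs, deferring to the parent implementation when the cast fails.
; BeginPlay and Tick are disabled stubs.
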(event EventBeginPlay
  (|CallParentFunction) ; This node is disabled and will not be called.\nDrag off pins to build functionality.
)

(event EventTick
  (params (DeltaSeconds :type Float%20%28single-precision%29))
  (|CallParentFunction DeltaSeconds) ; This node is disabled and will not be called.\nDrag off pins to build functionality.
)

(fn GetCost
  (bind _asfort_pickup_creative (Utilities|Casting|CastToFortPickupCreative (Components|GetOwner))
    (:then
      (Utilities|IsValid (Variables|/Script/FortniteGame.FortPickupCreative|GetCostComponent _asfort_pickup_creative)
        (:Is%20Valid
          (return (Utilities|Map|MakeMap (Utilities|Enum|LiteralenumEFortResourceType "Wood") (Utilities|Array|Get(aref) (Cost|GetItemCosts (Variables|/Script/FortniteGame.FortPickupCreative|GetCostComponent _asfort_pickup_creative)) (Math|Conversions|ToInteger(Byte) (Utilities|Enum|LiteralenumEFortResourceType "Wood"))) (Utilities|Enum|LiteralenumEFortResourceType "Stone") (Utilities|Array|Get(aref) (Cost|GetItemCosts (Variables|/Script/FortniteGame.FortPickupCreative|GetCostComponent _asfort_pickup_creative)) (Math|Conversions|ToInteger(Byte) (Utilities|Enum|LiteralenumEFortResourceType "Stone"))) (Utilities|Enum|LiteralenumEFortResourceType "Metal") (Utilities|Array|Get(aref) (Cost|GetItemCosts (Variables|/Script/FortniteGame.FortPickupCreative|GetCostComponent _asfort_pickup_creative)) (Math|Conversions|ToInteger(Byte) (Utilities|Enum|LiteralenumEFortResourceType "Metal"))) (Utilities|Enum|LiteralenumEFortResourceType "GoldCurrency") (Utilities|Array|Get(aref) (Cost|GetItemCosts (Variables|/Script/FortniteGame.FortPickupCreative|GetCostComponent _asfort_pickup_creative)) (Math|Conversions|ToInteger(Byte) (Utilities|Enum|LiteralenumEFortResourceType "GoldCurrency"))))))))
    (:CastFailed
      (return (|CallParentFunction)))))
