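; Parent Blueprint for the Creative item spawner device: it stores dropped items and spawns them as pickups during a minigame.
; Outside minigames, server-side hits on the drop box queue items, play FX, then move them into the spawner component's list.
; User options feed that component, which spawns after an optional delay and is paused by disable events or flagged blocked by overlapping pawns.
; Replicated state drives item preview scale and light-ray visibility.
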(event EventOnPlayModeChanged
  (params (Minigame :type Fort%20Minigame%20Object%20Reference) (bIsInPlayMode :type Boolean))
  ;[1 0.1082 0 0.25] Disable Collision on drop area during Edit mode, as overlaps are expensive.
  (|CallParentFunction Minigame bIsInPlayMode)
  (Collision|SetCollisionEnabled (Variables|Default|GetDropBox) (select bIsInPlayMode "NoCollision" "QueryOnly")))

(event EventInitializeDevice
  ;[0.1134 1 0] Initialization
  (|CallParentFunction)
  (Initialization|InitializeComponent (Variables|Default|GetCreative_ItemPreview_Component) (Variables|Default|GetRotationSceneComponent) (Variables|Default|GetPreviewStaticMesh) (Variables|Default|GetPreviewSkeletalMesh) (Variables|Default|GetHolo%20Loop%20Cue) 0 (Variables|Default|GetItemVariantNotifier) :!class /Game/Creative/Devices/Common/Components/Creative_ItemPreview_Component.Creative_ItemPreview_Component_C)
  (Class|CreativeItemPreviewComponent|UpdateScale (Variables|Default|GetCreative_ItemPreview_Component) (Variables|Default|GetItemSpawnScale))
  (Class|MinigameSpawnerComponent|SetPickupSettings (Variables|Default|GetMinigame_Spawner_Component) (Variables|Default|GetAllowedTeam) (Variables|Default|GetAllowedClass) (Variables|Default|GetKnob_AffectsTeam) (Variables|Default|GetKnob_AffectsClass) (Variables|Default|GetItemsVisibleToAllowedOnly)))

(event EventBeginPlay
  ;[0.1134 1 0] Initialization
  (|CallParentFunction)
  (Networking|SwitchHasAuthority
    (:Authority
      ;[1 1 1] Binding dispatcher events from the Spawner Component
      (bind _minigame_spawner_component (Variables|Default|GetMinigame_Spawner_Component)
        (:then
          (Default|BindEventtoAddedtoSpawnListDispatcher _minigame_spawner_component (delegate Custom|OnAddedToSpawnList))
          (Default|BindEventtoOnSpawnDispatcher _minigame_spawner_component (delegate Custom|OnSpawn))
          (Default|BindEventtoOnPickedUpSpawnedItem _minigame_spawner_component (delegate Custom|Picked%20Up%20Item))
          (OptionData|BindEventtoOnAnyPropertyChanged (Variables|Default|GetToyOptionsComponent) (EventDispatchers|CreateEvent :!event ResaveActor))
          (OptionData|AssignOnPropertiesLoaded (Variables|Default|GetToyOptionsComponent) (EventDispatchers|CreateEvent :!event ResaveActor)))
        (:else)))))

(event Custom|OnSpawn
  (params (Spawned%20Actor :type Actor%20Object%20Reference))
  ;[1 1 1] Binding dispatcher events from the Spawner Component
  (CallFunction|OnSpawnFX Spawned%20Actor)
  (Transformation|SetActorRotation Spawned%20Actor (Math|Rotator|MakeRotator 0.0 0.0 (.yaw (Transformation|GetActorRotation))) false)
  (if (Variables|SpawnSettings|GetbContinuouslyDeploy)
    (bind _desiredpickuplocation (CallFunction|GetDesiredPickupLocation))
    (Transformation|SetActorLocation Spawned%20Actor _desiredpickuplocation false false)
    (bind _asfort_pickup_creative (Utilities|Casting|CastToFortPickupCreative Spawned%20Actor))
    (Pickup|SetCombinePickupswhenTossCompletes _asfort_pickup_creative true)))

(event Custom|OnAddedToSpawnList
  (params (Added%20Actor :type Actor%20Object%20Reference))
  ;[1 1 1] Binding dispatcher events from the Spawner Component
  (CallFunction|ResaveActor))

(event Custom|Picked%20Up%20Item
  (params (Controller :type Controller%20Object%20Reference))
  ;[1 1 1] Binding dispatcher events from the Spawner Component
  (GameplayEvent|BroadcastEvent (Variables|UserOptions|Events|GetOn%20Item%20Picked%20Up) 0 (Utilities|Struct|MakeGameplayEventItemSpawnerPickedUpItem Controller)))

(event MinigameStartedDelegate(FortMinigameProgress)
  (bind _minigame_spawner_component (Variables|Default|GetMinigame_Spawner_Component))
  (bind _enabledonstart (Variables|Default|GetEnabledOnStart))
  (bind _initialspawndelay (Variables|SpawnSettings|GetInitialSpawnDelay))
  ;[0 1 0.175] When minigame starts
  (Utilities|FlowControl|Sequence
    (:then_0
      (Networking|SwitchHasAuthority
        (:Authority
          (Variables|Default|SetItemSpawnLocation (+ (Math|Vector|RotateVector "0, 0,100.000000" (Transformation|GetActorRotation)) (Transformation|GetWorldLocation (Variables|Default|GetDropBox))))
          (CallFunction|InitializeDefaultSpawnList)
          ;[1 1 1] Initial Spawn Timer setup
          (Variables|Default|SetbDisabledSpawning (not _enabledonstart))
          (Variables|Default|SetNo%20Remaining%20Spawns (== (Utilities|Array|Length (Variables|/Script/FortniteGame.FortMinigameSpawnerComponent|GetToSpawnList (Variables|Default|GetMinigame_Spawner_Component))) 0))
          (if _enabledonstart
            (Variables|/Game/Athena/Items/Traps/MinigameSpawner/Minigame_Spawner_Component.Minigame_Spawner_Component_C|SetbPauseSpawning false _minigame_spawner_component)
            (else
              (Class|MinigameSpawnerComponent|PauseSpawner _minigame_spawner_component)))
          (if (> _initialspawndelay 0.0)
            (bind _returnvalue (Utilities|Time|SetTimerbyEvent (delegate Custom|SetupMinigameStartSpawn) _initialspawndelay))
            (bind _output_get (Variables|Default|SetMinigameStartSpawnTimer _returnvalue))
            (if (Variables|Default|GetbDisabledSpawning)
              (Utilities|Time|PauseTimerbyHandle _output_get))
            (elif (and (== (Variables|SpawnSettings|GetInitialSpawnDelay) 0.0) (Variables|Default|GetEnabledOnStart))
              (Utilities|Time|ClearandInvalidateTimerbyHandle (Variables|Default|GetMinigameStartSpawnTimer))
              (bind _binprogress (CallFunction|IsMinigameinProgress))
              (if _binprogress
                (Class|MinigameSpawnerComponent|SpawnRemaining (Variables|Default|GetMinigame_Spawner_Component) true)))))))
    (:then_1
      (CallFunction|HandleonMinigameStarted))))

(event Custom|SetupMinigameStartSpawn
  ;[1 1 1] Initial Spawn Timer setup
  (Utilities|Time|ClearandInvalidateTimerbyHandle (Variables|Default|GetMinigameStartSpawnTimer))
  (bind _binprogress (CallFunction|IsMinigameinProgress))
  (if _binprogress
    (Class|MinigameSpawnerComponent|SpawnRemaining (Variables|Default|GetMinigame_Spawner_Component) true)))

(event MinigameEndedDelegate(FortMinigameProgress)
  ;[1 0.1233 0] When minigame ends
  (Utilities|FlowControl|Sequence
    (:then_0
      (Networking|SwitchHasAuthority
        (:Authority
          (Class|MinigameSpawnerComponent|ResetSpawner (Variables|Default|GetMinigame_Spawner_Component) true))))
    (:then_1
      (Collision|SetCollisionEnabled (Variables|Debug|GetStaticMeshComponent) "QueryAndPhysics"))
    (:then_2
      (CallFunction|HandleMinigameEnded))))

(event EventDestroyed
  (bind _holo_loop_cue (Variables|Default|GetHolo%20Loop%20Cue))
  ;[1 0 0.02512] Destruction
  (if (Audio|Components|Audio|IsPlaying _holo_loop_cue)
    (Audio|Components|Audio|FadeOut _holo_loop_cue 0.2 0.0 "Linear" :!class /Script/Engine.AudioComponent)))

(event OnComponentHit(DropBox)
  (params (HitComponent :type Primitive%20Component%20Object%20Reference) (OtherActor :type Actor%20Object%20Reference) (OtherComp :type Primitive%20Component%20Object%20Reference) (NormalImpulse :type Vector :default "0, 0, 0") (Hit :type Hit%20Result%20Structure%20%28by%20ref%29))
  (bind _itemstobeaddedqueue (Variables|Default|GetItemsToBeAddedQueue))
  (bind _itemstobeaddedqueue_2 (Variables|Default|GetItemsToBeAddedQueue))
  ;[0 0.5614 1 0.2] Handles item being added from a hit event.  Hit captures all pick-ups
  (Networking|SwitchHasAuthority
    (:Authority
      (bind _binprogress (CallFunction|IsMinigameinProgress))
      (if (not _binprogress)
        (bind _output_get (Variables|Default|SetItemThatHit OtherActor))
        (bind _asfort_spawnable_item_interface (Utilities|Casting|CastToFortSpawnableItemInterface _output_get))
        (if (SpawnableItemInterface|CanbeSpawned _asfort_spawnable_item_interface (Variables|Default|GetInvalidPickupGameplayTagContainer) self)
          ;;[1 1 1] Every bounce of the pickup causes a hit trigger.  This ensures rest of script only fires once
          (if (not (Utilities|Equal(Object) (Variables|Default|GetItemThatHit) (Variables|Default|GetItemToBeAdded)))
            (bind _output_get_2 (Variables|Default|SetItemToBeAdded (Variables|Default|GetItemThatHit)))
            (bind _returnvalue (Utilities|Array|Add _itemstobeaddedqueue _output_get_2))
            (CallFunction|RPCMULTICASTAddItemFX (Utilities|Array|Get(acopy) _itemstobeaddedqueue _returnvalue))
            (Utilities|FlowControl|Delay 2.0) ; Ensure object still exists so FX can start before object is destroyed
            ;;[1 1 1] Loop through Queue array to be sure to add every item if items are added in rapid succession
            (if (> (Utilities|Array|Length _itemstobeaddedqueue_2) 0)
              (CallFunction|AddtoSpawnList (Utilities|Array|Get(acopy) _itemstobeaddedqueue_2)))))))))

(event OnComponentBeginOverlap(BlockSpawnBox)
  (params (OverlappedComponent :type Primitive%20Component%20Object%20Reference) (OtherActor :type Actor%20Object%20Reference) (OtherComp :type Primitive%20Component%20Object%20Reference) (OtherBodyIndex :type Integer) (bFromSweep :type Boolean) (SweepResult :type Hit%20Result%20Structure%20%28by%20ref%29))
  ;[0 0.5614 1 0.2] Handles item being added from a hit event.  Hit captures all pick-ups
  ;;[0 0.4352 0.8594 0] Overlaps on Block Spawn Box
  (if (not (Variables|SpawnSettings|GetbSpawnWhenPawnBlocking))
    (bind _asfort_pawn (Utilities|Casting|CastToFortPawn OtherActor))
    (Utilities|Set|Add (Variables|Default|GetPawnsBlockingSpawner) _asfort_pawn)
    (Class|MinigameSpawnerComponent|SetSpawnerBlockedbyPawn (Variables|Default|GetMinigame_Spawner_Component) true)))

(event OnComponentEndOverlap(BlockSpawnBox)
  (params (OverlappedComponent :type Primitive%20Component%20Object%20Reference) (OtherActor :type Actor%20Object%20Reference) (OtherComp :type Primitive%20Component%20Object%20Reference) (OtherBodyIndex :type Integer))
  (bind _pawnsblockingspawner (Variables|Default|GetPawnsBlockingSpawner))
  ;[0 0.5614 1 0.2] Handles item being added from a hit event.  Hit captures all pick-ups
  ;;[0 0.4352 0.8594 0] Overlaps on Block Spawn Box
  (if (not (Variables|SpawnSettings|GetbSpawnWhenPawnBlocking))
    (bind _asfort_pawn (Utilities|Casting|CastToFortPawn OtherActor))
    (Utilities|Set|Remove _pawnsblockingspawner _asfort_pawn)
    (Class|MinigameSpawnerComponent|SetSpawnerBlockedbyPawn (Variables|Default|GetMinigame_Spawner_Component) (> (Utilities|Set|Length _pawnsblockingspawner) 0))))

(event Custom|Enable%20Event
  (params (Instigator :type Controller%20Object%20Reference))
  ;[0.9614 0 1] Gameplay Functions
  (Networking|SwitchHasAuthority
    (:Authority
      (bind _binprogress (CallFunction|IsMinigameinProgress))
      (if _binprogress
        (if (Variables|Default|GetbDisabledSpawning)
          (Variables|Default|SetbDisabledSpawning false)
          (CallFunction|HandleSpawnerEnabled))))))

(event Custom|Disable%20Event
  (params (Instigator :type Controller%20Object%20Reference))
  ;[0.9614 0 1] Gameplay Functions
  (Networking|SwitchHasAuthority
    (:Authority
      (bind _binprogress (CallFunction|IsMinigameinProgress))
      (if _binprogress
        (if (not (Variables|Default|GetbDisabledSpawning))
          (Variables|Default|SetbDisabledSpawning true)
          (CallFunction|HandleSpawnerDisabled))))))

(event OnLastItemPickedUp(Minigame_Spawner_Component)
  (Variables|Default|SetNo%20Remaining%20Spawns true))

(event OnItemContainerUpdated(Minigame_Spawner_Component)
  (CallFunction|UpdateItemPreviewList))

(event Custom|OnSpawnFX
  (params (Spawned%20Actor :type Actor%20Object%20Reference))
  ;[1 1 1] Fires when an item is spawned during the minigame.  Minigame must be active if you want to see/hear it occur.
  (if (Variables|Default|GetDisplayVFX)
    (Networking|CanExecuteCosmeticEvents
      (:True
        (Audio|PlaySoundatLocation (Variables|Default|GetItemSpawnSound) (Variables|Default|GetItemSpawnLocation)) ; ItemSpawnSound
        (Effects|SpawnEmitteratLocation "/Game/Athena/Items/Traps/MinigameSpawner/P_ItemSpawner_CastItem.P_ItemSpawner_CastItem" (Transformation|GetWorldLocation (Variables|Default|GetLightRays)) (Transformation|GetWorldRotation (Variables|Debug|GetStaticMeshComponent)) "1.000000,1.000000,1.000000" true "AutoRelease")))))

(event Custom|UpdateSpawnerFX
  ;[1 1 1] Show or hide emissive elements based on spawning being enabled
  (Networking|CanExecuteCosmeticEvents
    (:True
      (Development|SetHiddeninGame (Variables|Default|GetLightRays) (not (Math|Boolean|ANDBoolean (not (Variables|Default|GetbDisabledSpawning)) (Math|Conversions|ToBoolean(Integer) (Utilities|Array|Length (Variables|/Script/FortniteGame.FortMinigameSpawnerComponent|GetToSpawnList (Variables|Default|GetMinigame_Spawner_Component)))) (not (Variables|Default|GetNo%20Remaining%20Spawns)) (Variables|Default|GetDisplayVFX))) true))))

(event Custom|RPC_MULTICAST%20Add%20Item%20FX
  (params (Item :type Actor%20Object%20Reference))
  (Networking|CanExecuteCosmeticEvents
    (:True
      (bind _asfort_pickup_athena (Utilities|Casting|CastToFortPickupAthena Item))
      (Class|CreativeItemPreviewComponent|ItemAddedToQueueFX (Variables|Default|GetCreative_ItemPreview_Component) _asfort_pickup_athena))))

(fn AddToSpawnList
  (params (InputPin :type Actor%20Object%20Reference))
  (Networking|SwitchHasAuthority
    (:Authority
      (bind _wasvalid (Class|MinigameSpawnerComponent|AddtoSpawnList (Variables|Default|GetMinigame_Spawner_Component) InputPin (Variables|Default|GetItemSpawnLocation)))
      (if _wasvalid
        (Actor|DestroyActor InputPin) ; Destroy the pick-up that is in the world.  Hidden by FX
        (else
          _))
      (Utilities|Array|RemoveIndex (Variables|Default|GetItemsToBeAddedQueue)))))

(fn GetBaseSpawnLocation
  (return (+ (Transformation|GetWorldLocation (Variables|Default|GetRotationSceneComponent)) (Variables|Default|GetHoverLocationOffset))))

(fn GetDesiredPickupLocation
  (bind _base_spawn_location (CallFunction|GetBaseSpawnLocation))
  (if (and (Variables|SpawnSettings|GetbContinuouslyDeploy) (> (Utilities|Array|Length (Variables|/Game/Athena/Items/Traps/MinigameSpawner/Minigame_Spawner_Component.Minigame_Spawner_Component_C|GetSpawnedPickups (Variables|Default|GetMinigame_Spawner_Component))) 1))
    (return (+ (Math|Vector|MakeVector (* (Math|Random|RandomFloat) (Variables|Default|GetRandomSpawnDistance)) (* (Math|Random|RandomFloat) (Variables|Default|GetRandomSpawnDistance))) (- _base_spawn_location (Math|Vector|MakeVector (/ (Variables|Default|GetRandomSpawnDistance) 2.0) (/ (Variables|Default|GetRandomSpawnDistance) 2.0)))))
    (else
      (return _base_spawn_location))))

(fn HandleMinigameEnded
  (Networking|SwitchHasAuthority
    (:Authority
      (Utilities|Time|ClearandInvalidateTimerbyHandle (Variables|Default|GetMinigameStartSpawnTimer)))))

(fn HandleOnMinigameStarted)

(fn HandleSpawnerDisabled
  (Class|MinigameSpawnerComponent|PauseSpawner (Variables|Default|GetMinigame_Spawner_Component))
  (Utilities|Time|PauseTimerbyHandle (Variables|Default|GetMinigameStartSpawnTimer)))

(fn HandleSpawnerEnabled
  (bind _minigamestartspawntimer (Variables|Default|GetMinigameStartSpawnTimer))
  (bind _minigame_spawner_component (Variables|Default|GetMinigame_Spawner_Component))
  (if (Utilities|Time|IsTimerPausedbyHandle _minigamestartspawntimer)
    (Utilities|Time|UnpauseTimerbyHandle _minigamestartspawntimer)
    (elif (or (>= (Variables|SpawnSettings|GetInitialSpawnDelay) 0) (Class|MinigameSpawnerComponent|HasBegunSpawning (Variables|Default|GetMinigame_Spawner_Component)))
      (Class|MinigameSpawnerComponent|ResumeSpawner _minigame_spawner_component)))
  (Variables|/Game/Athena/Items/Traps/MinigameSpawner/Minigame_Spawner_Component.Minigame_Spawner_Component_C|SetbPauseSpawning false _minigame_spawner_component))

(fn HandleSpawnItemWhenReceived
  (if (not (Class|MinigameSpawnerComponent|HasBegunSpawning (Variables|Default|GetMinigame_Spawner_Component)))
    (Utilities|Time|ClearandInvalidateTimerbyHandle (Variables|Default|GetMinigameStartSpawnTimer))
    (Class|MinigameSpawnerComponent|SpawnRemaining (Variables|Default|GetMinigame_Spawner_Component) (not (Variables|Default|GetbDisabledSpawning)))
    (else
      (Class|MinigameSpawnerComponent|SpawnNext (Variables|Default|GetMinigame_Spawner_Component)))))

(fn InitializeDefaultSpawnList
  ;[1 1 1] Handles setting up the spawn list to account for items that were put into the list from UnrealEd defaults
  (bind _result (Math|Integer|DecrementInt (Utilities|Array|Length (Variables|/Script/FortniteGame.FortMinigameSpawnerComponent|GetToSpawnList (Variables|Default|GetMinigame_Spawner_Component))))
    (:%20%20
      (for _index (range (+ _result 1))
        (Utilities|Struct|SetmembersinMinigameSpawnerSpawnParams (Utilities|Array|Get(aref) (Variables|/Script/FortniteGame.FortMinigameSpawnerComponent|GetToSpawnList (Variables|Default|GetMinigame_Spawner_Component)) _index) (Math|Transform|MakeTransform (Variables|Default|GetItemSpawnLocation)))))))

(fn IsMinigameInProgress
  (bind _fortminigameprogress (Variables|Default|GetFortMinigameProgress))
  (bind _currentminigame (Variables|/Script/FortniteGame.FortMinigameLogicComponent|GetCurrentMinigame _fortminigameprogress))
  (Utilities|FlowControl|Sequence
    (:then_0
      (Utilities|IsValid _currentminigame
        (:Is%20Valid
          (LocalVariables|SetbSuccess (Minigame|HasStarted _currentminigame :!class /Script/FortniteGame.FortMinigame)))))
    (:then_1
      (return (LocalVariables|GetbSuccess)))))

(fn IsVisibleDuringMinigame
  (return (Variables|SpawnSettings|GetIsMeshVisibleDuringMinigame)))

(fn OnRep_bDisabledSpawning
  (CallFunction|UpdateSpawnerFX))

(fn OnRep_ItemSpawnScale
  (Networking|CanExecuteCosmeticEvents
    (:True
      (Class|CreativeItemPreviewComponent|UpdateScale (Variables|Default|GetCreative_ItemPreview_Component) (Variables|Default|GetItemSpawnScale)))))

(fn OnRep_NoRemainingSpawns
  (CallFunction|UpdateSpawnerFX))

(fn OnSpawnItemReceived
  (params (Instigator :type Controller%20Object%20Reference))
  (Networking|SwitchHasAuthority
    (:Authority
      (if (not (Variables|Default|GetbDisabledSpawning))
        (bind _binprogress (CallFunction|IsMinigameinProgress))
        (if _binprogress
          (CallFunction|HandleSpawnItemwhenReceived))))))

(fn ResaveActor
  (bind _minigame_spawner_component (Variables|Default|GetMinigame_Spawner_Component))
  (Variables|/Game/Athena/Items/Traps/MinigameSpawner/Minigame_Spawner_Component.Minigame_Spawner_Component_C|SetDelayBetweenSpawns (Variables|SpawnSettings|GetTimeBetweenSpawnsSeconds) _minigame_spawner_component)
  (Variables|/Game/Athena/Items/Traps/MinigameSpawner/Minigame_Spawner_Component.Minigame_Spawner_Component_C|SetbIsLooping (Variables|SpawnSettings|GetIsLooping) _minigame_spawner_component)
  (Variables|/Game/Athena/Items/Traps/MinigameSpawner/Minigame_Spawner_Component.Minigame_Spawner_Component_C|SetGrantExtraAmmo (Variables|SpawnSettings|GetGrantExtraAmmoSetting) _minigame_spawner_component)
  (Variables|/Game/Athena/Items/Traps/MinigameSpawner/Minigame_Spawner_Component.Minigame_Spawner_Component_C|SetRandomSpawnOption (Variables|SpawnSettings|GetRandomOption) _minigame_spawner_component)
  (Variables|/Game/Athena/Items/Traps/MinigameSpawner/Minigame_Spawner_Component.Minigame_Spawner_Component_C|SetIsPickupOnTouch (Variables|SpawnSettings|GetIsPickupOnTouch) _minigame_spawner_component)
  (Variables|/Game/Athena/Items/Traps/MinigameSpawner/Minigame_Spawner_Component.Minigame_Spawner_Component_C|SetItemSpawnScale (Math|Vector|MakeVector (Variables|Default|GetItemSpawnScale) (Variables|Default|GetItemSpawnScale) (Variables|Default|GetItemSpawnScale)) _minigame_spawner_component)
  (Variables|/Game/Athena/Items/Traps/MinigameSpawner/Minigame_Spawner_Component.Minigame_Spawner_Component_C|SetItemWoodCost (Variables|SpawnSettings|GetItemCostWood) _minigame_spawner_component)
  (Variables|/Game/Athena/Items/Traps/MinigameSpawner/Minigame_Spawner_Component.Minigame_Spawner_Component_C|SetItemMetalCost (Variables|SpawnSettings|GetItemCostMetal) _minigame_spawner_component)
  (Variables|/Game/Athena/Items/Traps/MinigameSpawner/Minigame_Spawner_Component.Minigame_Spawner_Component_C|SetItemStoneCost (Variables|SpawnSettings|GetItemCostStone) _minigame_spawner_component)
  (Variables|/Game/Athena/Items/Traps/MinigameSpawner/Minigame_Spawner_Component.Minigame_Spawner_Component_C|SetItemGoldCost (Variables|SpawnSettings|GetItemCostGold) _minigame_spawner_component)
  (Variables|/Game/Athena/Items/Traps/MinigameSpawner/Minigame_Spawner_Component.Minigame_Spawner_Component_C|SetNumSimultaneous (select (Variables|SpawnSettings|GetbContinuouslyDeploy) 500 1) _minigame_spawner_component)
  (CallFunction|Resave))

(fn ToSpawnListHasItems
  (return (> (Utilities|Array|Length (Variables|/Script/FortniteGame.FortMinigameSpawnerComponent|GetToSpawnList (Variables|Default|GetMinigame_Spawner_Component))) 0)))

(fn UpdateItemPreviewList
  (Networking|CanExecuteCosmeticEvents
    (:True
      (Class|CreativeItemPreviewComponent|UpdateItemListFromSpawnParams (Variables|Default|GetCreative_ItemPreview_Component) (Variables|/Script/FortniteGame.FortMinigameSpawnerComponent|GetToSpawnList (Variables|Default|GetMinigame_Spawner_Component)))
      (CallFunction|UpdateSpawnerFX))))

(fn ConstructionScript
  (|CallParentFunction))
