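; Creative item spawner device that previews and spawns item pickups through preview and minigame spawner components.
; On authority during an active minigame, cycling either starts spawning or destroys current pickups and spawns the next, unless spawning is disabled.
; Option changes map the initial-movement setting to pickup toss and gravity flags.
; They also reposition the preview, either atop the base mesh using scaled pickup capsule height or at a fixed height.
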(event EditorPreview|EventSetupForEditorPreview
  (bind _previewstaticmesh (Variables|Default|GetPreviewStaticMesh))
  ;[1 0.9002 0 0.25] Show the first item in the spawner during UEFN preview
  (|CallParentFunction)
  (Transformation|SetWorldScale3D _previewstaticmesh (Math|Conversions|ToVector(Float) (Variables|Default|GetItemSpawnScale)))
  (Class|CreativeItemPreviewComponent|EditorDisplayPreview (Variables|Default|GetCreative_ItemPreview_Component) _previewstaticmesh))

(event EventBeginPlay
  ;[0 1 0.2776 0.25] Initialization
  (Components|Tick|SetComponentTickEnabled (Variables|Default|GetPreviewSkeletalMesh) false)
  (|CallParentFunction)
  (Initialization|InitializeComponent (Variables|Default|GetCreative_ItemPreview_Component) (Variables|Default|GetRotationSceneComponent) (Variables|Default|GetPreviewStaticMesh) (Variables|Default|GetPreviewSkeletalMesh) (Variables|Default|GetHolo%20Loop%20Cue) 0 (Variables|Default|GetItemVariantNotifier) :!class /Game/Creative/Devices/Common/Components/Creative_ItemPreview_Component.Creative_ItemPreview_Component_C))

(event Custom|CycleToNextItem
  (params (Instigator :type Controller%20Object%20Reference))
  (Networking|SwitchHasAuthority
    (:Authority
      (if (not (Variables|Default|GetbDisabledSpawning))
        (bind _binprogress (CallFunction|IsMinigameinProgress))
        (if _binprogress
          (if (not (Class|MinigameSpawnerComponent|HasBegunSpawning (Variables|Default|GetMinigame_Spawner_Component)))
            (Utilities|Time|ClearandInvalidateTimerbyHandle (Variables|Default|GetMinigameStartSpawnTimer))
            (Class|MinigameSpawnerComponent|SpawnRemaining (Variables|Default|GetMinigame_Spawner_Component) (not (Variables|Default|GetbDisabledSpawning)))
            (else
              (Class|MinigameSpawnerComponent|DestroyPickups (Variables|Default|GetMinigame_Spawner_Component))
              (Class|MinigameSpawnerComponent|SpawnNext (Variables|Default|GetMinigame_Spawner_Component)))))))))

(event Custom|SpawnItem
  (params (Instigator :type Controller%20Object%20Reference))
  (CallFunction|OnSpawnItemReceived Instigator))

(event OnAnyPropertyChanged(ToyOptionsComponent)
  (Variables|/Game/Athena/Items/Traps/MinigameSpawner/Minigame_Spawner_Component.Minigame_Spawner_Component_C|SetbShouldTossPickups (Utilities|Enum|Equal(Enum) (Variables|Default|GetItemInitialMovement) "NewEnumerator0") (Variables|Default|GetMinigame_Spawner_Component))
  (Variables|/Game/Athena/Items/Traps/MinigameSpawner/Minigame_Spawner_Component.Minigame_Spawner_Component_C|SetbShouldApplyGravityToPickups (or (Utilities|Enum|Equal(Enum) (Variables|Default|GetItemInitialMovement) "NewEnumerator0") (Utilities|Enum|Equal(Enum) (Variables|Default|GetItemInitialMovement) "NewEnumerator1")) (Variables|Default|GetMinigame_Spawner_Component))
  (Transformation|SetRelativeLocation (Variables|Default|GetRotationSceneComponent) (CallFunction|GetItemPreviewLocationOffset) false false))

(event OnPropertiesLoaded(ToyOptionsComponent)
  (Variables|/Game/Athena/Items/Traps/MinigameSpawner/Minigame_Spawner_Component.Minigame_Spawner_Component_C|SetbShouldTossPickups (Utilities|Enum|Equal(Enum) (Variables|Default|GetItemInitialMovement) "NewEnumerator0") (Variables|Default|GetMinigame_Spawner_Component))
  (Variables|/Game/Athena/Items/Traps/MinigameSpawner/Minigame_Spawner_Component.Minigame_Spawner_Component_C|SetbShouldApplyGravityToPickups (or (Utilities|Enum|Equal(Enum) (Variables|Default|GetItemInitialMovement) "NewEnumerator0") (Utilities|Enum|Equal(Enum) (Variables|Default|GetItemInitialMovement) "NewEnumerator1")) (Variables|Default|GetMinigame_Spawner_Component))
  (Transformation|SetRelativeLocation (Variables|Default|GetRotationSceneComponent) (CallFunction|GetItemPreviewLocationOffset) false false))

(fn ActivateFromOtherDevice
  (params (Instigator :type Controller%20Object%20Reference) (Requester :type ERequesterType%20Enum :default "NewEnumerator0"))
  (return false))

(fn CanInteract
  (params (InteractingPawn :type Fort%20Pawn%20Object%20Reference) (InteractionBeingAttempted :type EInteractionBeingAttempted%20Enum :default "FirstInteraction") (InteractionType :type TInteractionType%20Enum :default "IT_NoInteraction"))
  (return false))

(fn GetDeviceColor
  (return 0))

(fn GetInstigatingController
  (return 0))

(fn GetCreativeActorBounds
  (bind (_origin _boxextent) (Collision|GetComponentBounds (Variables|Debug|GetStaticMeshComponent)))
  (return true _boxextent))

(fn GetCreativeActorOrigin
  (return true "0, 0, 0"))

(fn GetCreativeActorRelativeTransform
  (return false 0))

(fn GetGridAxisGuideAttachmentPoint
  (return false false "0, 0, 0"))

(fn GetItemPreviewLocationOffset
  (bind (_returnvalue_origin _returnvalue_boxextent) (Math|BoxSphereBounds|BreakBoxSphereBounds (StaticMesh|GetBounds (Variables|/Script/Engine.StaticMeshComponent|GetStaticMesh (Variables|Debug|GetStaticMeshComponent)))))
  (bind (_returnvalue_origin_x _returnvalue_origin_y _returnvalue_origin_z) (Math|Vector|BreakVector _returnvalue_origin))
  (bind (_returnvalue_boxextent_x _returnvalue_boxextent_y _returnvalue_boxextent_z) (Math|Vector|BreakVector _returnvalue_boxextent))
  (return 0.0 0.0 (Math|Float|SelectFloat (Utilities|Operators|Add (* (Variables|Default|GetItemSpawnScale) (Class|FortPickup|GetMovementCapsuleHalfHeight)) _returnvalue_origin_z _returnvalue_boxextent_z) 1e+02 (Utilities|Enum|Equal(Enum) (Variables|Default|GetItemInitialMovement) "NewEnumerator1"))))

(fn GetSpawnedPickups
  (return 0))

(fn IsEnabled
  (return (not (Variables|Default|GetbDisabledSpawning))))

(fn IsHoldingItemForPawn
  (params (Item :type Item%20Variant%20Handle%20Structure) (Pawn :type Pawn%20Object%20Reference))
  (return false))

(fn SetTriggerCollision
  (params (On :type Boolean))
  (return false))

(fn ConstructionScript
  (|CallParentFunction))
